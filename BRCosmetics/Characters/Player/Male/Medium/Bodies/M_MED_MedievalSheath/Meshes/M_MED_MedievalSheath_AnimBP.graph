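# Post-process style AnimBP for the MedievalSheath player body cosmetic, adding secondary motion to its pouches, arm bands, knee pads, back flaps and scarf.
# It copies the attached parent's pose, applies weighted bone constraints, then runs a Control Rig fed parachute, frontend and emote flags.
# AnimDynamics springs then simulate flaps, scarf (constrained and nudged first) and arm bands.
# On customization load, frontend state selects scarf gravity; its sim-space and alpha variables are unread here.
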AnimBlueprint /BRCosmetics/Characters/Player/Male/Medium/Bodies/M_MED_MedievalSheath/Meshes/M_MED_MedievalSheath_AnimBP
target_skeleton=&/M_MED_MedievalSheath_Skeleton
graphs [
 AnimGraph {nodes [
   "Constraint - dyn_backFlap_c_01 (P)"(Alpha=1, ConstraintWeights_0=1) {
    alphaBoolBlend {blendOption=Linear}
    boneToModify {boneName=dyn_backFlap_c_01}
    constraintSetup [{targetBone {boneName=pelvis}}]
    constraintWeights [1]
   }
   n1 = Anim Dynamics - Bone: dyn_backFlap_l!(
    SimulationSpace:AnimPhysSimSpaceType Enum=World
    ComponentPose=Local To Component(LocalPose=Use cached pose 'CtrlRigCache'())
    Alpha=1
    ExternalForce="0,0,0"
   ) -> (Pose) {
    angularDampingOverride=0.99
    angularSpringConstant=1000
    bAngularSpring=true
    bGravityOverrideInSimSpace=true
    bOverrideAngularDamping=true
    bUseGravityOverride=true
    alphaBoolBlend {blendOption=Linear}
    boundBone {boneName=dyn_backFlap_l}
    componentAppliedLinearAccClamp=<1000 1000 1000>
    componentLinearAccScale=<1 1 1>
    componentLinearVelScale=<1 1 1>
    gravityOverride {z=-980}
    physicsBodyDefinitions [{
      boundBone {boneName=dyn_backFlap_l}
      boxExtents=<15 15 15>
      constraintSetup {angularLimitsMax=<0 0 0>, angularLimitsMin=<0 -20 0>, angularTarget=<0 0 0>}
      localJointOffset=<1 0 0>
     }]
    relativeSpaceBone {boneName=root}
   }
   n2 = Anim Dynamics - Bone: dyn_backFlap_c_01!(SimulationSpace:AnimPhysSimSpaceType Enum=World, ComponentPose=@n1, Alpha=1, ExternalForce="0,0,0") -> (Pose) {
    angularDampingOverride=0.99
    angularSpringConstant=1000
    bAngularSpring=true
    bGravityOverrideInSimSpace=true
    bOverrideAngularDamping=true
    bUseGravityOverride=true
    alphaBoolBlend {blendOption=Linear}
    boundBone {boneName=dyn_backFlap_c_01}
    componentAppliedLinearAccClamp=<1000 1000 1000>
    componentLinearAccScale=<1 1 1>
    componentLinearVelScale=<1 1 1>
    gravityOverride {z=-980}
    physicsBodyDefinitions [{
      boundBone {boneName=dyn_backFlap_c_01}
      boxExtents=<15 15 15>
      constraintSetup {angularLimitsMax=<0 0 0>, angularLimitsMin=<0 -20 0>, angularTarget=<0 0 0>}
      localJointOffset=<1 0 0>
     }]
    relativeSpaceBone {boneName=root}
   }
   n3 = Anim Dynamics - Bone: dyn_backFlap_c_02!(SimulationSpace:AnimPhysSimSpaceType Enum=World, ComponentPose=@n2, Alpha=1, ExternalForce="0,0,0") -> (Pose) {
    angularDampingOverride=0.99
    angularSpringConstant=1000
    bAngularSpring=true
    bGravityOverrideInSimSpace=true
    bOverrideAngularDamping=true
    bUseGravityOverride=true
    alphaBoolBlend {blendOption=Linear}
    boundBone {boneName=dyn_backFlap_c_02}
    componentAppliedLinearAccClamp=<1000 1000 1000>
    componentLinearAccScale=<1 1 1>
    componentLinearVelScale=<1 1 1>
    gravityOverride {z=-980}
    physicsBodyDefinitions [{
      boundBone {boneName=dyn_backFlap_c_02}
      boxExtents=<15 15 15>
      constraintSetup {angularLimitsMax=<0 0 0>, angularLimitsMin=<0 -20 0>, angularTarget=<0 0 0>}
      localJointOffset=<1 0 0>
     }]
    relativeSpaceBone {boneName=root}
   }
   n4 = Anim Dynamics - Bone: dyn_backFlap_r!(SimulationSpace:AnimPhysSimSpaceType Enum=World, ComponentPose=@n3, Alpha=1, ExternalForce="0,0,0") -> (Pose) {
    angularDampingOverride=0.99
    angularSpringConstant=1000
    bAngularSpring=true
    bGravityOverrideInSimSpace=true
    bOverrideAngularDamping=true
    bUseGravityOverride=true
    alphaBoolBlend {blendOption=Linear}
    boundBone {boneName=dyn_backFlap_r}
    componentAppliedLinearAccClamp=<1000 1000 1000>
    componentLinearAccScale=<1 1 1>
    componentLinearVelScale=<1 1 1>
    gravityOverride {z=-980}
    physicsBodyDefinitions [{
      boundBone {boneName=dyn_backFlap_r}
      boxExtents=<15 15 15>
      constraintSetup {angularLimitsMax=<0 0 0>, angularLimitsMin=<0 -20 0>, angularTarget=<0 0 0>}
      localJointOffset=<1 0 0>
     }]
    relativeSpaceBone {boneName=root}
   }
   n5 = "Constraint - dyn_scarf_01 (PPPPPPPPP)"!(
    ComponentPose=@n4
    Alpha=1
    ConstraintWeights_0=0.0651
    ConstraintWeights_1:float=0.0992
    ConstraintWeights_2:float=0.5581
    ConstraintWeights_3:float=0.0171
    ConstraintWeights_4:float=0.0409
    ConstraintWeights_5:float=0.058
    ConstraintWeights_6:float=0.0061
    ConstraintWeights_7:float=0.1513
    ConstraintWeights_8:float=0.0043
   ) -> (Pose) {
    alphaBoolBlend {blendOption=Linear}
    boneToModify {boneName=dyn_scarf_01}
    constraintSetup [
     {targetBone {boneName=spine_03}}
     {targetBone {boneName=spine_04}}
     {targetBone {boneName=spine_05}}
     {targetBone {boneName=neck_01}}
     {targetBone {boneName=neck_02}}
     {targetBone {boneName=clavicle_l}}
     {targetBone {boneName=upperarm_l}}
     {targetBone {boneName=upperarm_twist_01_l}}
     {targetBone {boneName=upperarm_twist_02_l}}
    ]
    constraintWeights [1,1,1,1,1,1,1,1,1]
   }
   n6 = "Constraint - dyn_scarf_02 (PPPPP)"!(
    ComponentPose=@n5
    Alpha=1
    ConstraintWeights_0=0.011
    ConstraintWeights_1:float=0.1079
    ConstraintWeights_2:float=0.4176
    ConstraintWeights_3:float=0.4629
    ConstraintWeights_4:float=0.0006
   ) -> (Pose) {
    alphaBoolBlend {blendOption=Linear}
    boneToModify {boneName=dyn_scarf_02}
    constraintSetup [
     {targetBone {boneName=spine_02}}
     {targetBone {boneName=spine_03}}
     {targetBone {boneName=spine_04}}
     {targetBone {boneName=spine_05}}
     {targetBone {boneName=neck_02}}
    ]
    constraintWeights [1,1,1,1,1]
   }
   n7 = "Transform (Modify) Bone - Bone: dyn_scarf_01"(ComponentPose=@n6, Alpha=1, Translation="0,0,1", Rotation="5,0,0", Scale="1,1,1") {
    rotationMode=BMM_Additive
    rotationSpace=BCS_BoneSpace
    translationMode=BMM_Additive
    translationSpace=BCS_BoneSpace
    alphaBoolBlend {blendOption=Linear}
    boneToModify {boneName=dyn_scarf_01}
    rotation {pitch=5}
    translation {z=1}
   }
   n8 = Get Gravity Override()
   n9 = Anim Dynamics - Bone: dyn_scarf_01!(SimulationSpace:AnimPhysSimSpaceType Enum=BoneRelative, ComponentPose=@n7, Alpha=1, GravityOverride:Vector=@n8, ExternalForce="0,0,0") -> (Pose) {
    angularDampingOverride=0.9999
    angularSpringConstant=5000
    bAngularSpring=true
    bGravityOverrideInSimSpace=true
    bOverrideAngularDamping=true
    bUseGravityOverride=true
    simulationSpace=BoneRelative
    alphaBoolBlend {blendOption=Linear}
    boundBone {boneName=dyn_scarf_01}
    componentAppliedLinearAccClamp=<1000 1000 1000>
    componentLinearAccScale=<1 1 1>
    componentLinearVelScale=<1 1 1>
    gravityOverride {z=-5000}
    physicsBodyDefinitions [{
      boundBone {boneName=dyn_scarf_01}
      boxExtents=<15 15 15>
      constraintSetup {angularLimitsMax=<0 0 0>, angularLimitsMin=<0 -20 0>, angularTarget=<0 0 0>}
      localJointOffset=<0.1 0 0>
     }]
    relativeSpaceBone {boneName=spine_05}
   }
   n10 = Anim Dynamics - Bone: dyn_scarf_02!(SimulationSpace:AnimPhysSimSpaceType Enum=BoneRelative, ComponentPose=@n9, Alpha=1, GravityOverride:Vector=@n8, ExternalForce="0,0,0") -> (Pose) {
    angularDampingOverride=0.9999
    angularSpringConstant=5000
    bAngularSpring=true
    bGravityOverrideInSimSpace=true
    bOverrideAngularDamping=true
    bUseGravityOverride=true
    simulationSpace=BoneRelative
    alphaBoolBlend {blendOption=Linear}
    boundBone {boneName=dyn_scarf_02}
    componentAppliedLinearAccClamp=<1000 1000 1000>
    componentLinearAccScale=<1 1 1>
    componentLinearVelScale=<1 1 1>
    gravityOverride {z=-5000}
    physicsBodyDefinitions [{
      boundBone {boneName=dyn_scarf_02}
      boxExtents=<15 15 15>
      constraintSetup {angularLimitsMax=<0 0 0>, angularLimitsMin=<0 -20 0>, angularTarget=<0 0 0>}
      localJointOffset=<0.1 0 0>
     }]
    relativeSpaceBone {boneName=spine_05}
   }
   n11 = Anim Dynamics - Bone: dyn_armBand_top!(SimulationSpace:AnimPhysSimSpaceType Enum=Component, ComponentPose=@n10, Alpha=1, ExternalForce="0,0,0") -> (Pose) {
    angularDampingOverride=0.9999
    angularSpringConstant=5000
    bAngularSpring=true
    bGravityOverrideInSimSpace=true
    bOverrideAngularDamping=true
    bUseGravityOverride=true
    simulationSpace=BoneRelative
    alphaBoolBlend {blendOption=Linear}
    boundBone {boneName=dyn_armBand_top}
    componentAppliedLinearAccClamp=<1000 1000 1000>
    componentLinearAccScale=<1 1 1>
    componentLinearVelScale=<1 1 1>
    gravityOverride {z=-980}
    physicsBodyDefinitions [{
      boundBone {boneName=dyn_armBand_top}
      boxExtents=<2 2 2>
      constraintSetup {angularLimitsMax=<5 10 10>, angularLimitsMin=<-5 -20 -5>, angularTarget=<0 0 0>}
      localJointOffset=<0 0.5 0>
     }]
    relativeSpaceBone {boneName=root}
   }
   n12 = Anim Dynamics - Bone: dyn_armBand_bot!(SimulationSpace:AnimPhysSimSpaceType Enum=BoneRelative, ComponentPose=@n11, Alpha=1, ExternalForce="0,0,0") -> (Pose) {
    angularDampingOverride=0.9999
    angularSpringConstant=5000
    bAngularSpring=true
    bGravityOverrideInSimSpace=true
    bOverrideAngularDamping=true
    bUseGravityOverride=true
    simulationSpace=BoneRelative
    alphaBoolBlend {blendOption=Linear}
    boundBone {boneName=dyn_armBand_bot}
    componentAppliedLinearAccClamp=<1000 1000 1000>
    componentLinearAccScale=<1 1 1>
    componentLinearVelScale=<1 1 1>
    gravityOverride {z=-980}
    physicsBodyDefinitions [{
      boundBone {boneName=dyn_armBand_bot}
      boxExtents=<5 5 5>
      constraintSetup {angularLimitsMax=<5 10 10>, angularLimitsMin=<-5 -10 -10>, angularTarget=<0 0 0>}
      localJointOffset=<0 0 0.5>
     }]
    relativeSpaceBone {boneName=root}
   }
   Output Pose(Result=Component To Local(ComponentPose=@n12)) {layerGroup=None}
   n13 = "Constraint - dyn_pouch_l_01 (PPPPPPP)"!(
    ComponentPose=Local To Component(LocalPose=Copy Pose From Mesh() {bUseAttachedParent=true})
    Alpha=1
    ConstraintWeights_0=0.6253
    ConstraintWeights_1:float=0.2843
    ConstraintWeights_2:float=0.0558
    ConstraintWeights_3:float=0.0275
    ConstraintWeights_4:float=0.0009
    ConstraintWeights_5:float=0.0031
    ConstraintWeights_6:float=0.003
   ) -> (Pose) {
    alphaBoolBlend {blendOption=Linear}
    boneToModify {boneName=dyn_pouch_l_01}
    constraintSetup [
     {targetBone {boneName=pelvis}}
     {targetBone {boneName=spine_01}}
     {targetBone {boneName=spine_02}}
     {targetBone {boneName=spine_03}}
     {targetBone {boneName=spine_04}}
     {targetBone {boneName=thigh_twist_01_l}}
     {targetBone {boneName=thigh_twist_01_r}}
    ]
    constraintWeights [1,1,1,1,1,1,1]
   }
   n14 = "Constraint - dyn_pouch_l_02 (PPPPPPP)"!(
    ComponentPose=@n13
    Alpha=1
    ConstraintWeights_0=0.4999
    ConstraintWeights_1:float=0.3537
    ConstraintWeights_2:float=0.0963
    ConstraintWeights_3:float=0.0462
    ConstraintWeights_4:float=0.0025
    ConstraintWeights_5:float=0.0007
    ConstraintWeights_6:float=0.0007
   ) -> (Pose) {
    alphaBoolBlend {blendOption=Linear}
    boneToModify {boneName=dyn_pouch_l_02}
    constraintSetup [
     {targetBone {boneName=pelvis}}
     {targetBone {boneName=spine_01}}
     {targetBone {boneName=spine_02}}
     {targetBone {boneName=spine_03}}
     {targetBone {boneName=spine_04}}
     {targetBone {boneName=thigh_twist_01_l}}
     {targetBone {boneName=thigh_twist_01_r}}
    ]
    constraintWeights [1,1,1,1,1,1,1]
   }
   n15 = "Constraint - dyn_pouch_l_03 (PPPPPPP)"!(
    ComponentPose=@n14
    Alpha=1
    ConstraintWeights_0=0.6325
    ConstraintWeights_1:float=0.2739
    ConstraintWeights_2:float=0.0509
    ConstraintWeights_3:float=0.0253
    ConstraintWeights_4:float=0.0013
    ConstraintWeights_5:float=0.0074
    ConstraintWeights_6:float=0.0086
   ) -> (Pose) {
    alphaBoolBlend {blendOption=Linear}
    boneToModify {boneName=dyn_pouch_l_03}
    constraintSetup [
     {targetBone {boneName=pelvis}}
     {targetBone {boneName=spine_01}}
     {targetBone {boneName=spine_02}}
     {targetBone {boneName=spine_03}}
     {targetBone {boneName=spine_04}}
     {targetBone {boneName=thigh_twist_01_l}}
     {targetBone {boneName=thigh_twist_01_r}}
    ]
    constraintWeights [1,1,1,1,1,1,1]
   }
   n16 = "Constraint - dyn_pouch_l_04 (PPPPPPP)"!(
    ComponentPose=@n15
    Alpha=1
    ConstraintWeights_0=0.6325
    ConstraintWeights_1:float=0.2739
    ConstraintWeights_2:float=0.0509
    ConstraintWeights_3:float=0.0253
    ConstraintWeights_4:float=0.0013
    ConstraintWeights_5:float=0.0074
    ConstraintWeights_6:float=0.0086
   ) -> (Pose) {
    alphaBoolBlend {blendOption=Linear}
    boneToModify {boneName=dyn_pouch_l_04}
    constraintSetup [
     {targetBone {boneName=pelvis}}
     {targetBone {boneName=spine_01}}
     {targetBone {boneName=spine_02}}
     {targetBone {boneName=spine_03}}
     {targetBone {boneName=spine_04}}
     {targetBone {boneName=thigh_twist_01_l}}
     {targetBone {boneName=thigh_twist_01_r}}
    ]
    constraintWeights [1,1,1,1,1,1,1]
   }
   n17 = "Constraint - dyn_pouch_r_01 (PPPPPPP)"!(
    ComponentPose=@n16
    Alpha=1
    ConstraintWeights_0=0.5075
    ConstraintWeights_1:float=0.3497
    ConstraintWeights_2:float=0.0939
    ConstraintWeights_3:float=0.0451
    ConstraintWeights_4:float=0.0024
    ConstraintWeights_5:float=0.0011
    ConstraintWeights_6:float=0.0004
   ) -> (Pose) {
    alphaBoolBlend {blendOption=Linear}
    boneToModify {boneName=dyn_pouch_r_01}
    constraintSetup [
     {targetBone {boneName=pelvis}}
     {targetBone {boneName=spine_01}}
     {targetBone {boneName=spine_02}}
     {targetBone {boneName=spine_03}}
     {targetBone {boneName=spine_04}}
     {targetBone {boneName=thigh_twist_01_l}}
     {targetBone {boneName=thigh_twist_01_r}}
    ]
    constraintWeights [1,1,1,1,1,1,1]
   }
   n18 = "Constraint - dyn_pouch_r_02 (PPPPPPP)"!(
    ComponentPose=@n17
    Alpha=1
    ConstraintWeights_0=0.6325
    ConstraintWeights_1:float=0.2739
    ConstraintWeights_2:float=0.0509
    ConstraintWeights_3:float=0.0253
    ConstraintWeights_4:float=0.0013
    ConstraintWeights_5:float=0.0074
    ConstraintWeights_6:float=0.0086
   ) -> (Pose) {
    alphaBoolBlend {blendOption=Linear}
    boneToModify {boneName=dyn_pouch_r_02}
    constraintSetup [
     {targetBone {boneName=pelvis}}
     {targetBone {boneName=spine_01}}
     {targetBone {boneName=spine_02}}
     {targetBone {boneName=spine_03}}
     {targetBone {boneName=spine_04}}
     {targetBone {boneName=thigh_twist_01_l}}
     {targetBone {boneName=thigh_twist_01_r}}
    ]
    constraintWeights [1,1,1,1,1,1,1]
   }
   n19 = "Constraint - dyn_pouch_r_03 (PPPPPPP)"!(
    ComponentPose=@n18
    Alpha=1
    ConstraintWeights_0=0.6325
    ConstraintWeights_1:float=0.2739
    ConstraintWeights_2:float=0.0509
    ConstraintWeights_3:float=0.0253
    ConstraintWeights_4:float=0.0013
    ConstraintWeights_5:float=0.0074
    ConstraintWeights_6:float=0.0086
   ) -> (Pose) {
    alphaBoolBlend {blendOption=Linear}
    boneToModify {boneName=dyn_pouch_r_03}
    constraintSetup [
     {targetBone {boneName=pelvis}}
     {targetBone {boneName=spine_01}}
     {targetBone {boneName=spine_02}}
     {targetBone {boneName=spine_03}}
     {targetBone {boneName=spine_04}}
     {targetBone {boneName=thigh_twist_01_l}}
     {targetBone {boneName=thigh_twist_01_r}}
    ]
    constraintWeights [1,1,1,1,1,1,1]
   }
   n20 = "Constraint - dyn_pouch_r_04 (PPPPPPP)"!(
    ComponentPose=@n19
    Alpha=1
    ConstraintWeights_0=0.6768
    ConstraintWeights_1:float=0.2474
    ConstraintWeights_2:float=0.0358
    ConstraintWeights_3:float=0.0183
    ConstraintWeights_4:float=0.0008
    ConstraintWeights_5:float=0.0096
    ConstraintWeights_6:float=0.0113
   ) -> (Pose) {
    alphaBoolBlend {blendOption=Linear}
    boneToModify {boneName=dyn_pouch_r_04}
    constraintSetup [
     {targetBone {boneName=pelvis}}
     {targetBone {boneName=spine_01}}
     {targetBone {boneName=spine_02}}
     {targetBone {boneName=spine_03}}
     {targetBone {boneName=spine_04}}
     {targetBone {boneName=thigh_twist_01_l}}
     {targetBone {boneName=thigh_twist_01_r}}
    ]
    constraintWeights [1,1,1,1,1,1,1]
   }
   n21 = "Constraint - dyn_armBand_top (PPPPPPP)"!(
    ComponentPose=@n20
    Alpha=1
    ConstraintWeights_0=0.0055
    ConstraintWeights_1:float=0.251
    ConstraintWeights_2:float=0.0545
    ConstraintWeights_3:float=0.6655
    ConstraintWeights_4:float=0.0004
    ConstraintWeights_5:float=0.0009
    ConstraintWeights_6:float=0.0222
   ) -> (Pose) {
    alphaBoolBlend {blendOption=Linear}
    boneToModify {boneName=dyn_armBand_top}
    constraintSetup [
     {targetBone {boneName=clavicle_r}}
     {targetBone {boneName=upperarm_r}}
     {targetBone {boneName=upperarm_twist_01_r}}
     {targetBone {boneName=upperarm_twist_02_r}}
     {targetBone {boneName=lowerarm_r}}
     {targetBone {boneName=lowerarm_twist_02_r}}
     {targetBone {boneName=dyn_armBand_top}}
    ]
    constraintWeights [1,1,1,1,1,1,1]
   }
   n22 = "Constraint - dyn_armBand_bot (PPPPPPP)"!(
    ComponentPose=@n21
    Alpha=1
    ConstraintWeights_0=0.0021
    ConstraintWeights_1:float=0.2015
    ConstraintWeights_2:float=0.0266
    ConstraintWeights_3:float=0.7546
    ConstraintWeights_4:float=0.0031
    ConstraintWeights_5:float=0.0051
    ConstraintWeights_6:float=0.0071
   ) -> (Pose) {
    alphaBoolBlend {blendOption=Linear}
    boneToModify {boneName=dyn_armBand_bot}
    constraintSetup [
     {targetBone {boneName=clavicle_r}}
     {targetBone {boneName=upperarm_r}}
     {targetBone {boneName=upperarm_twist_01_r}}
     {targetBone {boneName=upperarm_twist_02_r}}
     {targetBone {boneName=lowerarm_r}}
     {targetBone {boneName=lowerarm_twist_02_r}}
     {targetBone {boneName=dyn_armBand_bot}}
    ]
    constraintWeights [1,1,1,1,1,1,1]
   }
   n23 = "Constraint - dyn_kneePad_l (PP)"!(ComponentPose=@n22, Alpha=1, ConstraintWeights_0=0.5, ConstraintWeights_1:float=0.5) -> (Pose) {
    alphaBoolBlend {blendOption=Linear}
    boneToModify {boneName=dyn_kneePad_l}
    constraintSetup [{targetBone {boneName=thigh_l}}, {targetBone {boneName=calf_l}}]
    constraintWeights [1,1]
   }
   n24 = "Constraint - dyn_kneePad_r (PP)"!(ComponentPose=@n23, Alpha=1, ConstraintWeights_0=0.5, ConstraintWeights_1:float=0.5) -> (Pose) {
    alphaBoolBlend {blendOption=Linear}
    boneToModify {boneName=dyn_kneePad_r}
    constraintSetup [{targetBone {boneName=thigh_r}}, {targetBone {boneName=calf_r}}]
    constraintWeights [1,1]
   }
   n25 = "Constraint - dyn_backFlap_l (PPP)"!(ComponentPose=@n24, Alpha=1, ConstraintWeights_0=0.6625, ConstraintWeights_1:float=0.0203, ConstraintWeights_2:float=0.3172) -> (Pose) {
    alphaBoolBlend {blendOption=Linear}
    boneToModify {boneName=dyn_backFlap_l}
    constraintSetup [{targetBone {boneName=pelvis}}, {targetBone {boneName=spine_01}}, {targetBone {boneName=thigh_twist_01_l}}]
    constraintWeights [1,1,1]
   }
   n26 = "Constraint - dyn_backFlap_c_01 (PPPPP)"!(
    ComponentPose=@n25
    Alpha=1
    ConstraintWeights_0=0.8637
    ConstraintWeights_1:float=0.0943
    ConstraintWeights_2:float=0.0085
    ConstraintWeights_3:float=0.0104
    ConstraintWeights_4:float=0.0231
   ) -> (Pose) {
    alphaBoolBlend {blendOption=Linear}
    boneToModify {boneName=dyn_backFlap_c_01}
    constraintSetup [
     {targetBone {boneName=pelvis}}
     {targetBone {boneName=spine_01}}
     {targetBone {boneName=spine_03}}
     {targetBone {boneName=thigh_twist_01_l}}
     {targetBone {boneName=thigh_twist_01_r}}
    ]
    constraintWeights [1,1,1,1,1]
   }
   n27 = "Constraint - dyn_backFlap_r (PPPPP)"!(
    ComponentPose=@n26
    Alpha=1
    ConstraintWeights_0=0.6529
    ConstraintWeights_1:float=0.0199
    ConstraintWeights_2:float=0
    ConstraintWeights_3:float=0.0001
    ConstraintWeights_4:float=0.3271
   ) -> (Pose) {
    alphaBoolBlend {blendOption=Linear}
    boneToModify {boneName=dyn_backFlap_r}
    constraintSetup [
     {targetBone {boneName=pelvis}}
     {targetBone {boneName=spine_01}}
     {targetBone {boneName=thigh_twist_01_l}}
     {targetBone {boneName=thigh_r}}
     {targetBone {boneName=thigh_twist_01_r}}
    ]
    constraintWeights [1,1,1,1,1]
   }
   Save cached pose 'ConstraintCache'(Pose=Component To Local(ComponentPose=@n27)) {cacheName=ConstraintCache}
   n28 = Control Rig!(
    Source=Use cached pose 'ConstraintCache'()
    Alpha
    is SkydiveParachute:bool=Get bIsParachuteOpened()
    Is In Frontend:bool=Get bIsInFrontEnd()
    Is Playing Emote:bool=Get bIsPlayingEmote()
   ) -> (Pose) {
    alphaBoolBlend {blendOption=Linear}
    controlRigAssetReference {blueprintRigClass=&/M_MED_MedievalSheath_CtrlRig.M_MED_MedievalSheath_CtrlRig_C}
   }
   Save cached pose 'CtrlRigCache'(Pose=@n28) {cacheName=CtrlRigCache}
  ]}
 EventGraph {nodes [
   Event OnCharacterCustomizationLoaded()
   > n1 = Branch(Condition=Get bIsInFrontEnd())
   > Set FE AnimDyn SimSpace(FE AnimDyn SimSpace:AnimPhysSimSpaceType Enum=World) -> (Output_Get:AnimPhysSimSpaceType Enum)
   > Set AnimDyn FE Alpha(AnimDyn FE Alpha:double=0.3) -> (Output_Get:double)
   > Set Gravity Override(Gravity Override="0,0,-5000")
   Set FE AnimDyn SimSpace(execute=@n1.else, FE AnimDyn SimSpace=BoneRelative)
   > Set AnimDyn FE Alpha(AnimDyn FE Alpha=0.7)
   > Set Gravity Override(Gravity Override="0,0,-980")
   Event BlueprintUpdateAnimation()
  ]}
]
variables [
 AnimDyn FE Alpha {type=number}
 FE AnimDyn SimSpace {type=AnimPhysSimSpaceType, default=BoneRelative}
 Gravity Override {type=Vector, default=<0 0 -5000>}
]
